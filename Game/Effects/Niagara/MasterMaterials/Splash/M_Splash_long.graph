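# Two-sided masked subsurface master material for water splash effects on Niagara mesh particles, tinted by particle color.
# Normal and mask textures share tiled UVs, warped by a radial flowmap through CenterWarp and optionally shifted in X by UVOffset.
# Particle alpha drives an erosion power curve on one mask channel, multiplied by a second channel for opacity.
# The MeshWPOBend dynamic parameter scales a Z-axis bend, weighted by cubed inverse V.
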Material /Game/Effects/Niagara/MasterMaterials/Splash/M_Splash_long
graph {
 properties {
  bAutomaticallySetUsageInEditor=false
  bUsedWithNiagaraMeshParticles=true
  blendMode=BLEND_Masked
  shadingModel=MSM_Subsurface
  twoSided=true
  usedShadingModels=Subsurface
 }
 nodes [
  Constant3Vector()
  Constant() {r=0.5}
  Constant()
  Constant() {r=0.2}
  n1 = VectorParameter() {parameterName=Tile, defaultValue=<rgba 2 1 0 0>}
  n2 = TextureCoordinate()
  n3 = TextureSample(UVs=MaterialFunctionCall(Texture Scale=Constant() {r=1.4}) {materialFunction=&ScaleUVsByCenter}.UVs) {samplerType=SAMPLERTYPE_LinearColor, texture=/Game/Environments/Hexmap/HexClouds/T_RadialFlowmap01}
  n4 = DynamicParameter() {defaultValue=<rgba 1 0 0 1>, paramNames [CenterWarp,MeshWPOBend,UVOffset,Param4]}
  n5 = Multiply(A=Subtract(A=ComponentMask(None=@n3.RGB) {g=true, r=true}) {constB=0.5}, B=@n4.CenterWarp)
  n6 = Constant2Vector() {r=1}
  n7 = Add(
   A=Multiply(A=AppendVector(A=@n1.R, B=@n1.R), B=Add(A=@n2, B=@n5))
   B=LinearInterpolate(B=Multiply(A=@n4.UVOffset, B=@n6), Alpha=ScalarParameter() {parameterName=UseXUVOffset})
  )
  n8 = TextureSampleParameter2D(UVs=@n7) {
   parameterName=Normal
   samplerType=SAMPLERTYPE_Normal
   texture=/Game/Packages/Fortress_Transylvania/Haunted_Forest/Swamps/Textures/T_water_splash_05_n
  }
  LinearInterpolate(A=@n8.RGB, B=Constant3Vector() {constant=<rgba 0 0 1 0>}, Alpha=ScalarParameter() {defaultValue=0.05, parameterName=Flatten Normal})
  n9 = ParticleColor()
  n10 = TextureSampleParameter2D(UVs=@n7) {parameterName=Mask, texture=/Game/Packages/Fortress_Transylvania/Haunted_Forest/Swamps/Textures/T_Water_Splash_Erode}
  n11 = ChannelMaskParameter(None=@n10.RGB) {maskChannel=Blue, parameterName=ErodeMask, defaultValue=<rgba 0 0 1 0>}
  n12 = LinearInterpolate(Alpha=@n9.A) {constA=2, constB=-0.1}
  Multiply(
   A=Multiply(A=@n9.A, B=Power(Base=@n11.R, Exp=@n12))
   B=ChannelMaskParameter(None=@n10.RGBA) {maskChannel=Green, parameterName=OpacityMask, defaultValue=<rgba 0 1 0 0>}.R
  )
  Multiply(A=@n9.RGB, B=ScalarParameter() {defaultValue=0.15, parameterName=EmissiveAdd}) {constB=0.15}
  n13 = ComponentMask(None=TextureCoordinate()) {g=true}
  Multiply(
   A=Multiply(
    A=Constant3Vector() {constant=<rgba 0 0 1 0>}
    B=DynamicParameter() {defaultValue=<rgba 1 0 0 1>, paramNames [CenterWarp,MeshWPOBend,UVOffset,Param4]}.MeshWPOBend
   )
   B=Power(Base=OneMinus(None=@n13)) {constExponent=3}
  )
  Multiply(A=VectorParameter() {group=Color, parameterName=Subsurface Color, defaultValue=<rgba 0.5 0.5 0.5 1>}.RGB, B=ParticleColor().RGB)
 ]
}
